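# Translucent unlit Niagara material for the glowing fragment effect on the FearlessFlightHero backpack.
# Bevel texture samples at slightly offset scales are thresholded into blue, green and magenta lines, layered over a faint variation-texture fill.
# Dynamic parameters set line and fragment strength and drive stepped radial sphere-mask growth.
# Stepped time animates a BumpOffset wobble on some quality levels; particle color tints emissive and scales opacity.
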Material /BRCosmetics/Accessories/FORT_Backpacks/Backpack_M_MED_FearlessFlightHero/FX/Materials/M_FearlessFlightHero_Fragment
graph {
 properties {
  bUsedWithNiagaraMeshParticles=true
  bUsedWithNiagaraSprites=true
  blendMode=BLEND_Translucent
  pixelDepthOffsetMode=PDOM_Legacy
  shadingModel=MSM_Unlit
  twoSided=true
  usedShadingModels=Unlit
 }
 nodes [
  n1 = TextureCoordinate()
  n2 = Constant2Vector() {g=0.5, r=0.5}
  n3 = DynamicParameter() {paramNames [LineStrenght, FragmentStrenght, LineGrowth, FragmentGrowth]}
  n4 = ScalarParameter() {defaultValue=12, parameterName=Stepped animation FPS}
  n5 = Floor(None=Multiply(A=@n3.LineGrowth, B=@n4) {constB=12})
  n6 = TextureCoordinate() {uTiling=1.5, vTiling=1.5}
  n7 = Time()
  n8 = ScalarParameter() {defaultValue=12, parameterName=Stepped animation FPS}
  n9 = Divide(A=Floor(None=Multiply(A=@n7, B=@n8) {constB=12}), B=@n8) {constB=12}
  n10 = TextureSample(UVs=@n6) {
   samplerType=SAMPLERTYPE_LinearGrayscale
   texture=/BRCosmetics/Accessories/FORT_Backpacks/Backpack_M_MED_FearlessFlightHero/FX/Textures/T_FFH_FragmentVar_M
  }
  n11 = BumpOffset(Coordinate=@n6, Height=Sine(None=Add(A=@n9, B=@n10.R)), HeightRatioInput=Constant() {r=0.01})
  n12 = QualitySwitch(Default=@n11, Low=@n6, High=@n11, Medium=@n6, Epic=@n11)
  n13 = TextureSample(UVs=@n12) {
   samplerType=SAMPLERTYPE_LinearGrayscale
   texture=&T_FFH_FragmentVar_M
  }
  n14 = Add(A=LinearInterpolate(Alpha=Divide(A=@n5, B=@n4) {constB=12}) {constA=-0.05, constB=0.5}, B=Multiply(A=@n13.R) {constB=0.05})
  n15 = Constant()
  n16 = Divide(A=Floor(None=Multiply(A=@n3.FragmentGrowth) {constB=12})) {constB=12}
  n17 = SphereMask(A=@n1, B=@n2, Radius=Add(A=LinearInterpolate(Alpha=@n16) {constA=-0.05, constB=0.5}, B=Multiply() {constB=0.05}), Hardness=@n15)
  n18 = Multiply(A=Sine(None=SphereMask(A=@n1, B=@n2, Radius=@n14, Hardness=@n15)), B=@n17)
  n19 = TextureSample(UVs=@n12) {
   samplerType=SAMPLERTYPE_LinearGrayscale
   texture=/BRCosmetics/Accessories/FORT_Backpacks/Backpack_M_MED_FearlessFlightHero/FX/Textures/T_FFH_FragmentBevel_M
  }
  n20 = Constant() {r=0.05}
  n21 = Sine(None=@n13.R)
  n22 = Multiply(A=@n18, B=Multiply(A=Step(Y=@n19.G, X=@n20), B=Saturate(None=@n21)))
  n23 = Sine(None=Add(A=@n13.R) {constB=0.1})
  n24 = TextureSample(UVs=MaterialFunctionCall(UVs=@n12, Texture Scale=Constant2Vector() {g=1.025, r=1.025}) {materialFunction=&ScaleUVsByCenter}.UVs) {
   samplerType=SAMPLERTYPE_LinearGrayscale
   texture=&T_FFH_FragmentBevel_M
  }
  n25 = Multiply(A=@n18, B=Multiply(A=Saturate(None=@n23), B=Step(Y=@n24.G, X=@n20)))
  n26 = Add(
   A=Multiply(A=Saturate(None=@n22), B=Constant3Vector() {constant=<rgba 0 0.3033 1 1>})
   B=Multiply(A=Saturate(None=@n25), B=Constant3Vector() {constant=<rgba 0 1 0.4918 1>})
  )
  n27 = Sine(None=Add(A=@n13.R) {constB=0.5})
  n28 = TextureSample(UVs=MaterialFunctionCall(UVs=@n12, Texture Scale=Constant2Vector() {g=0.98, r=0.98}) {materialFunction=&ScaleUVsByCenter}.UVs) {
   samplerType=SAMPLERTYPE_LinearGrayscale
   texture=&T_FFH_FragmentBevel_M
  }
  n29 = Multiply(A=@n18, B=Multiply(A=Saturate(None=@n27), B=Step(Y=@n28.G, X=@n20)))
  n30 = Multiply(A=Saturate(None=@n29), B=Constant3Vector() {constant=<rgba 0.8228 0 1 1>})
  n31 = DynamicParameter() {paramNames [LineStrenght, FragmentStrenght, LineGrowth, FragmentGrowth]}
  n32 = Multiply(A=Multiply(A=Add(A=@n26, B=@n30)) {constB=3}, B=@n31.LineStrenght)
  n33 = SmoothStep(Value=Multiply(A=@n17, B=@n19.G)) {constMax=0.02}
  n34 = Multiply(A=Multiply(A=Multiply(A=@n33, B=@n13.R)) {constB=0.5}, B=@n31.FragmentStrenght)
  n35 = ParticleColor()
  MaterialFunctionCall(Emissive=Multiply(A=Add(A=@n32, B=@n34), B=@n35.RGB)) {materialFunction=&MF_FX_EmissiveAdjustments}
  n36 = Multiply(A=Add(A=@n13.R) {constB=0.25}) {constB=0.2}
  n37 = Add(A=@n22, B=@n25)
  Multiply(A=Add(A=Multiply(A=@n17, B=@n36), B=Add(A=@n37, B=@n29)), B=@n35.A)
 ]
}
