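# Additive, unlit, two-sided Niagara sprite material for glowing hex-line effects on the FearlessFlightHero backpack cosmetic.
# A line mask and variation texture drive staggered linear-sine pulses that set brightness and, with particle alpha, opacity.
# Counter-panning smoke noise and the pulses blend purple to pink, hue-shifted by a dynamic parameter before shared emissive adjustments.
# Floored time gives the animation a stepped, low-frame-rate look.
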Material /BRCosmetics/Accessories/FORT_Backpacks/Backpack_M_MED_FearlessFlightHero/FX/Materials/M_FearlessFlightHero_LineHex
graph {
 properties {
  bUsedWithNiagaraSprites=true
  blendMode=BLEND_Additive
  pixelDepthOffsetMode=PDOM_Legacy
  shadingModel=MSM_Unlit
  twoSided=true
  usedShadingModels=Unlit
 }
 nodes [
  n1 = TextureSample() {
   samplerType=SAMPLERTYPE_LinearGrayscale
   texture=/BRCosmetics/Accessories/FORT_Backpacks/Backpack_M_MED_FearlessFlightHero/FX/Textures/T_FearlessFlightHero_Line_M
  }
  n2 = NamedRerouteUsage()
  n3 = TextureSample() {
   samplerType=SAMPLERTYPE_LinearColor
   texture=/BRCosmetics/Accessories/FORT_Backpacks/Backpack_M_MED_FearlessFlightHero/FX/Textures/T_FearlessFlightHero_LineVar_M
  }
  n4 = Add(
   A=MaterialFunctionCall(Value=@n2, Period=Constant() {r=2}) {materialFunction=&LinearSine}."Rounded Linear Sine "
   B=@n3.R
  )
  n5 = Constant() {r=4}
  n6 = Power(Base=Multiply(A=MaterialFunctionCall(Value=@n4, Period=@n5) {materialFunction=&LinearSine}."Rounded Linear Sine "))
  n7 = Multiply(A=SmoothStep(Value=@n1.R) {constMax=0.85, constMin=0.75}, B=@n6)
  n8 = Saturate(None=Multiply(A=@n7) {constB=0.5})
  n9 = Add(A=@n3.R, B=Add(A=@n2, B=@n3.G))
  n10 = Multiply(
   A=@n7
   B=Multiply(A=MaterialFunctionCall(Value=@n9) {materialFunction=&LinearSine}."Rounded Linear Sine ")
  )
  n11 = Saturate(None=Multiply(A=@n6, B=SmoothStep(Value=@n10) {constMax=0.6, constMin=0.25}))
  n12 = Add(A=Add(A=@n8, B=Multiply(A=@n11) {constB=0.4}), B=Power(Base=Multiply(A=@n1.R) {constB=0.5}) {constExponent=0.5})
  n13 = DynamicParameter() {paramNames [Hueshift,Param2,Param3,Param4]}
  n14 = TextureCoordinate() {uTiling=0.5, vTiling=0.5}
  n15 = NamedRerouteUsage()
  n16 = Add(
   A=TextureSample(UVs=Panner(Coordinate=@n14, Time=@n15) {speedX=-0.02, speedY=0.02}) {texture=&T_Inky_Smoke_Tile}.G
   B=TextureSample(UVs=Panner(Coordinate=@n14, Time=@n15) {speedX=0.02, speedY=-0.02}) {texture=&T_Inky_Smoke_Tile}.R
  )
  n17 = Multiply(A=@n16) {constB=8}
  n18 = Add(A=Sine(None=Add(A=@n17, B=@n2)) {period=4})
  n19 = Saturate(None=Add(A=@n16, B=Divide(A=@n18) {constB=4}))
  n20 = LinearInterpolate(
   A=Constant3Vector() {constant=<rgba 0.1626 0 0.6458 1>}
   B=Constant3Vector() {constant=<rgba 0.9911 0.0671 1 1>}
   Alpha=Power(Base=Add(A=@n19, B=@n8)) {constExponent=4}
  )
  MaterialFunctionCall(Emissive=Multiply(
    A=@n12
    B=MaterialFunctionCall(Hue Shift Percentage=@n13.Hueshift, Texture=@n20) {materialFunction=&HueShift}.Result
   )) {materialFunction=&MF_FX_EmissiveAdjustments}
  Multiply(A=@n12, B=ParticleColor().A)
  n21 = Constant() {r=8}
  n22 = Multiply(A=Time(), B=@n21) {constB=12}
  NamedRerouteDeclaration(None=Divide(A=Floor(None=@n22), B=@n21) {constB=12}) {name=TimeStep, nodeColor=<rgba 0.5703 0 1 1>}
 ]
}
