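# Additive, unlit, two-sided Niagara mesh-particle material that draws animated line sweeps for the FearlessFlightHero backpack effect.
# A thresholded Perlin mask, offset by dynamic RandomMask, flips V and varies speed for a swapped-UV panning gradient, sharpened by a power.
# Pan time is quantized to 12 steps per second with a RandomTime offset, giving stepped motion.
# A U-based sphere mask shapes the streaks before particle color and alpha tint them.
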Material /BRCosmetics/Accessories/FORT_Backpacks/Backpack_M_MED_FearlessFlightHero/FX/Materials/M_FearlessFlightHero_LineSweep
graph {
 properties {
  bUsedWithNiagaraMeshParticles=true
  blendMode=BLEND_Additive
  pixelDepthOffsetMode=PDOM_Legacy
  shadingModel=MSM_Unlit
  twoSided=true
  usedShadingModels=Unlit
 }
 nodes [
  n1 = ParticleColor()
  n2 = TextureCoordinate()
  n3 = ComponentMask(None=@n2) {g=true}
  n4 = TextureCoordinate() {uTiling=0, vTiling=0.5}
  n5 = DynamicParameter() {paramNames [RandomTime,RandomMask,Param3,Param4]}
  n6 = Step(
   Y=TextureSample(UVs=Panner(Coordinate=@n4, Time=@n5.RandomMask) {speedX=0.1}) {texture=/Game/Global/Textures/Default/Noise/3xPerlin_M}.R
   X=Constant() {r=0.25}
  )
  n7 = AppendVector(A=LinearInterpolate(A=@n3, B=OneMinus(None=@n3), Alpha=@n6), B=ComponentMask(None=@n2) {r=true})
  n8 = NamedRerouteUsage()
  n9 = LinearInterpolate(A=Constant() {r=0.1}, B=Constant() {r=0.2}, Alpha=@n6)
  n10 = Power(
   Base=TextureSample(UVs=Panner(Coordinate=@n7, Time=@n8, Speed=@n9) {speedX=1}) {texture=/BRCosmetics/Characters/Player/Female/Medium/Bodies/F_MED_Rays/Skins/Night/Textures/Gradient_Rays}.G
   Exp=Constant() {r=5}
  )
  n11 = Constant() {r=4}
  n12 = SphereMask(A=TextureCoordinate() {vTiling=0}, B=Constant2Vector() {r=0.5}, Radius=Constant() {r=0.5}, Hardness=Constant())
  n13 = Multiply(A=@n12) {constB=0.5}
  n14 = SmoothStep(Value=@n12) {constMin=0.5}
  n15 = Saturate(None=Multiply(A=Multiply(A=@n10, B=@n11), B=Add(A=@n13, B=@n14)))
  MaterialFunctionCall(Emissive=Multiply(A=@n1.RGB, B=@n15)) {materialFunction=&MF_FX_EmissiveAdjustments}
  Multiply(A=@n15, B=@n1.A)
  n16 = Multiply(A=Add(A=Time(), B=@n5.RandomTime)) {constB=12}
  NamedRerouteDeclaration(None=Divide(A=Floor(None=@n16)) {constB=12}) {name=TimeStep, nodeColor=<rgba 0.5703 0 1 1>}
 ]
}
