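# Translucent unlit Niagara sprite material that draws animated lines inside a hexagon for the Fearless Flight Hero back bling effects.
# Tangent-space camera vector offsets UVs for view-dependent parallax depth, then polar-style coordinates feed two panned samples of one lines texture on stepped, frame-rate-quantized time.
# View angle blends the samples; a polygon mask, center cutout and facing fade shape opacity, scaled by particle alpha, while particle color drives emissive.
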Material /BRCosmetics/Accessories/FORT_Backpacks/Backpack_M_MED_FearlessFlightHero/FX/Materials/M_FearlessFlightHero_Lines
graph {
 properties {bUsedWithNiagaraSprites=true, blendMode=BLEND_Translucent, pixelDepthOffsetMode=PDOM_Legacy, shadingModel=MSM_Unlit, usedShadingModels=Unlit}
 nodes [
  n1 = ParticleColor()
  MaterialFunctionCall(Emissive=@n1.RGB) {materialFunction=&MF_FX_EmissiveAdjustments}
  n2 = Transform(None=CameraVectorWS()) {transformSourceType=TRANSFORMSOURCE_World, transformType=TRANSFORM_Tangent}
  n3 = Multiply(A=ComponentMask(None=@n2) {g=true, r=true}) {constB=-1}
  n4 = ComponentMask(None=@n2) {b=true}
  n5 = ScalarParameter() {defaultValue=0.5, parameterName=Parallax Depth A}
  n6 = ScalarParameter() {defaultValue=1.5, parameterName=Parallax Depth B}
  n7 = CameraVectorWS()
  n8 = VertexNormalWS()
  n9 = Saturate(None=Abs(None=DotProduct(A=@n7, B=@n8)))
  n10 = OneMinus(None=Power(Base=OneMinus(None=@n9), Exp=ScalarParameter() {defaultValue=0.333, parameterName=Texture Blend Exponent}))
  n11 = Add(
   A=ConstantBiasScale(None=TextureCoordinate()) {bias=-0.5, scale=2}
   B=Multiply(A=Divide(A=@n3, B=@n4) {constB=1}, B=LinearInterpolate(A=@n5, B=@n6, Alpha=@n10))
  ) {constB=0}
  n12 = MaterialFunctionCall(Float2=@n11) {materialFunction=&BreakOutFloat2Components}
  n13 = Constant2Vector() {g=113, r=710}
  n14 = Add(A=Divide(A=Arctangent2Fast(Y=@n12.G, X=@n12.G), B=Divide(A=@n13.R, B=@n13.R))) {constB=0.5}
  n15 = Distance(A=@n11, B=Constant())
  n16 = Power(Base=OneMinus(None=@n15))
  n17 = AppendVector(A=@n14, B=Add(A=OneMinus(None=@n16), B=@n14))
  n18 = VectorParameter() {parameterName=Texture Tiling, defaultValue=<rgba 9 1 2 0.2>}
  n19 = Multiply(A=AppendVector(A=@n17, B=@n17), B=AppendVector(A=@n18.RGB, B=@n18.RGB))
  n20 = Time()
  n21 = ScalarParameter() {defaultValue=8, parameterName=Stepped animation FPS}
  n22 = Divide(A=Floor(None=Multiply(A=@n20, B=@n21) {constB=12}), B=@n21) {constB=12}
  n23 = VectorParameter() {parameterName=Panner Speeds, defaultValue=<rgba 0.033 0.3 -0.025 0.333>}
  n24 = TextureSample(UVs=Panner(Coordinate=ComponentMask(None=@n19) {g=true, r=true}, Time=@n22, Speed=AppendVector(A=@n23.R, B=@n23.R))) {
   samplerType=SAMPLERTYPE_LinearGrayscale
   texture=/BRCosmetics/Accessories/FORT_Backpacks/Backpack_M_MED_FearlessFlightHero/FX/Textures/T_Backpack_FearlessFlightHero_Lines
  }
  n25 = Power(Base=@n24.R, Exp=ScalarParameter() {defaultValue=6, parameterName=Texture Exponent A})
  n26 = TextureSample(UVs=Panner(Coordinate=ComponentMask(None=@n19) {a=true, b=true}, Time=@n22, Speed=AppendVector(A=@n23.B, B=@n23.B)) {speedY=-1}) {
   samplerType=SAMPLERTYPE_LinearGrayscale
   texture=&T_Backpack_FearlessFlightHero_Lines
  }
  n27 = Power(Base=@n26.R, Exp=ScalarParameter() {defaultValue=4.5, parameterName=Texture Exponent B})
  n28 = MaterialFunctionCall!(
   UVs
   Sides=Constant() {r=6}
   Size
   Roundness
   Rotation=ScalarParameter() {defaultValue=0.25, parameterName=Hexagon Rotation}
   Normal Intensity
  ) {materialFunction=/Game/Effects/Fort_Effects/MaterialFunctions/MF_Polygon}
  n29 = Saturate(None=@n15)
  n30 = Subtract(
   A=Multiply(A=LinearInterpolate(A=@n25, B=@n27, Alpha=@n10), B=SmoothStep(Value=@n28.Mask) {constMax=0.1}) {desc=hexagon mask}
   B=Power(Base=OneMinus(None=@n29)) {constExponent=5}
  )
  n31 = CameraVectorWS()
  n32 = VertexNormalWS()
  Multiply(A=Multiply(A=@n1.A, B=Saturate(None=@n30)), B=Abs(None=DotProduct(A=@n31, B=@n32)))
 ]
}
