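# Additive unlit Niagara mesh-particle material for the Fearless Flight backpack's hex volume effect.
# Emissive is particle color weighted by normal-to-camera facing, passed through MF_FX_EmissiveAdjustments.
# Opacity combines that facing with a smoothstepped rerouted mask and particle alpha, adding depth fade except on Low and Medium quality.
# Named reroutes build animated cell noise and a rotated dot pattern from particle-centered screen coordinates, with density driven by interpolation between parameters.
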Material /BRCosmetics/Accessories/FORT_Backpacks/Backpack_M_MED_FearlessFlightHero/FX/Materials/M_FearlessFlight_Hexvolume
graph {
 properties {
  bUsedWithNiagaraMeshParticles=true
  blendMode=BLEND_Additive
  pixelDepthOffsetMode=PDOM_Legacy
  shadingModel=MSM_Unlit
  usedShadingModels=Unlit
 }
 nodes [
  n1 = ParticleColor()
  n2 = DotProduct(A=VertexNormalWS(), B=CameraVectorWS())
  MaterialFunctionCall(Emissive=Multiply(A=Multiply(A=@n1.RGB, B=@n2), B=Saturate(None=@n2))) {materialFunction=&MF_FX_EmissiveAdjustments}
  n3 = NamedRerouteUsage()
  n4 = Add(A=Multiply(A=Power(Base=@n2) {constExponent=3}) {constB=0.25}, B=SmoothStep(Value=Multiply(A=@n2, B=@n3)) {constMin=0.5})
  n5 = Multiply(A=@n4, B=DepthFade() {fadeDistanceDefault=15})
  Multiply(A=Saturate(None=QualitySwitch(Default=@n5, Low=@n4, High=@n5, Medium=@n4, Epic=@n5)), B=ParticleColor().A)
  n6 = Multiply(A=NamedRerouteUsage()) {constB=6}
  n7 = Sine(None=DotProduct(A=Floor(None=@n6), B=Constant2Vector() {g=44.3, r=7.556})) {period=0.33}
  n8 = Multiply(A=@n7, B=Constant() {r=2.33})
  n9 = Multiply(A=Time(), B=ScalarParameter() {defaultValue=0.22, parameterName=DotBlendCellAnimTimeMult}) {constB=0.2}
  n10 = Multiply(A=@n7, B=Constant() {r=1.222})
  n11 = Multiply(A=Frac(None=Add(A=@n8, B=@n9)), B=Frac(None=Add(A=@n10, B=@n9)))
  NamedRerouteDeclaration(None=LinearInterpolate(A=Step(Y=@n11) {constX=0.25}, B=Step(Y=@n11) {constX=0.5})) {name=CellNoiseStepped, nodeColor=<rgba 0 0.6016 1 1>}
  NamedRerouteDeclaration(None=@n11) {name=CellNoiseNormal, nodeColor=<rgba 1 0 0.9141 1>}
  n12 = ParticlePositionWS()
  NamedRerouteDeclaration(None=Divide(A=MaterialFunctionCall!(Center Point World Position=@n12) {materialFunction=/Engine/Functions/Engine_MaterialFunctions02/UVs/WorldSpaceAlignedScreenCoordinates}.Offset UVs) {constB=100}) {name=ObjectAlighedScreenPosition, nodeColor=<rgba 1 0.867 0 1>}
  n13 = Subtract(A=Distance(A=ActorPositionWS(), B=CameraPositionWS())) {constB=100}
  NamedRerouteDeclaration(None=Saturate(None=Divide(A=@n13) {constB=400})) {name=DistanceBlending, nodeColor=<rgba 0.4766 1 0 1>}
  n14 = LinearInterpolate(
   A=ScalarParameter() {defaultValue=100, group=1_RoseDust, parameterName=DotsAmountMax, sliderMax=512, sortPriority=20}
   B=ScalarParameter() {defaultValue=66, parameterName=DotsAmountMin}
   Alpha=NamedRerouteUsage()
  )
  n15 = Multiply(A=NamedRerouteUsage(), B=LinearInterpolate(A=@n14, B=Multiply(A=@n14) {constB=0.5}, Alpha=NamedRerouteUsage())) {constA=0.5}
  n16 = Constant() {r=1}
  n17 = Add(A=Frac(None=Rotator(Coordinate=@n15, Time=@n16) {speed=0.75})) {constB=-0.5}
  n18 = Constant() {r=0.1}
  NamedRerouteDeclaration(None=OneMinus(None=Distance(A=@n17, B=@n18))) {name=DotBlend, nodeColor=<rgba 0 0.4844 1 1>}
 ]
}
