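# Translucent unlit effect material for Niagara mesh particles on the Fearless Flight Hero backpack, tinted by particle color and alpha.
# Particle-centered screen-aligned UVs drive rotated halftone dots, whose density varies by animated hashed cells and halves with camera distance.
# By default, the dots combine with bands cut from mesh UVs.
# VerticalLineSwitch instead uses panning stretched Perlin noise, adding hue-shifted emissive and dot-modulated opacity.
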Material /BRCosmetics/Accessories/FORT_Backpacks/Backpack_M_MED_FearlessFlightHero/FX/Materials/M_FearlessFlight_VolumeHex
graph {
 properties {
  bUsedWithNiagaraMeshParticles=true
  blendMode=BLEND_Translucent
  pixelDepthOffsetMode=PDOM_Legacy
  shadingModel=MSM_Unlit
  usedShadingModels=Unlit
 }
 nodes [
  n1 = TextureCoordinate()
  n2 = MaterialFunctionCall(Float2=Abs(None=Add(A=@n1) {constB=-0.5})) {materialFunction=&BreakOutFloat2Components}
  n3 = Constant() {r=0.5}
  n4 = Distance(A=@n2.G, B=@n3)
  n5 = OneMinus(None=@n4)
  n6 = Multiply(A=NamedRerouteUsage()) {constB=15}
  n7 = Sine(None=DotProduct(A=Floor(None=@n6), B=Constant2Vector() {g=44.3, r=7.556})) {period=0.33}
  n8 = Multiply(A=@n7, B=Constant() {r=2.33})
  n9 = Multiply(A=Time(), B=ScalarParameter() {defaultValue=0.22, parameterName=DotBlendCellAnimTimeMult}) {constB=0.2}
  n10 = Multiply(A=@n7, B=Constant() {r=1.222})
  n11 = Multiply(A=Frac(None=Add(A=@n8, B=@n9)), B=Frac(None=Add(A=@n10, B=@n9)))
  n12 = LinearInterpolate(A=Step(Y=@n11) {constX=0.25}, B=Step(Y=@n11) {constX=0.5})
  n13 = LinearInterpolate(
   A=ScalarParameter() {defaultValue=125, group=1_RoseDust, parameterName=DotsAmountMax, sliderMax=512, sortPriority=20}
   B=ScalarParameter() {defaultValue=75, parameterName=DotsAmountMin}
   Alpha=@n12
  )
  n14 = ActorPositionWS()
  n15 = CameraPositionWS()
  n16 = Divide(A=Subtract(A=Distance(A=@n14, B=@n15)) {constB=100}) {constB=400}
  n17 = Multiply(A=NamedRerouteUsage(), B=LinearInterpolate(A=@n13, B=Multiply(A=@n13) {constB=0.5}, Alpha=Saturate(None=@n16))) {constA=0.5}
  n18 = Constant() {r=1}
  n19 = Add(A=Frac(None=Rotator(Coordinate=@n17, Time=@n18) {speed=0.75})) {constB=-0.5}
  n20 = Constant() {r=0.1}
  n21 = Add(A=OneMinus(None=@n12), B=OneMinus(None=Distance(A=@n19, B=@n20)))
  n22 = Distance(A=@n2.R, B=@n3)
  n23 = Constant() {r=0.01}
  n24 = Step(Y=@n4, X=Constant() {r=0.1})
  n25 = Step(Y=Distance(A=@n2.G, B=Constant() {r=0.225}), X=Constant() {r=0.02})
  n26 = Add(A=Multiply(A=Add(A=@n5, B=@n21)) {constB=0.5}, B=Max(A=Step(Y=@n22, X=@n23), B=Add(A=@n24, B=@n25)))
  n27 = NamedRerouteUsage()
  n28 = Multiply(A=@n27, B=Constant2Vector() {g=0.1, r=1})
  n29 = Multiply(A=@n27, B=Constant2Vector() {g=0.1, r=6})
  n30 = Add(
   A=TextureSample(UVs=Panner(Coordinate=@n28) {speedX=0.1, speedY=0.1}) {texture=/Game/Items/Traps/Textures/T_PerlinRGB_01_M}.G
   B=TextureSample(UVs=Panner(Coordinate=@n29) {speedX=0.1, speedY=0.1}) {texture=&T_PerlinRGB_01_M}.R
  )
  n31 = StaticSwitchParameter(True=Power(Base=@n30) {constExponent=0.5}, False=@n26) {parameterName=VerticalLineSwitch}
  n32 = ParticleColor()
  n33 = Multiply(A=@n31, B=@n32.RGB)
  MaterialFunctionCall(Emissive=StaticSwitchParameter(
    True=MaterialFunctionCall(Hue Shift Percentage=@n31, Texture=@n33) {materialFunction=&HueShift}.Result
    False=@n33
   ) {parameterName=VerticalLineSwitch}) {materialFunction=&MF_FX_EmissiveAdjustments}
  n34 = ParticlePositionWS()
  NamedRerouteDeclaration(None=Divide(A=MaterialFunctionCall!(Center Point World Position=@n34) {materialFunction=/Engine/Functions/Engine_MaterialFunctions02/UVs/WorldSpaceAlignedScreenCoordinates}.Offset UVs) {constB=100}) {name=ScreenUv, nodeColor=<rgba 1 0.9844 0 1>}
  n35 = Multiply(A=@n30, B=Multiply(A=@n21) {constB=0.5})
  Reroute(None=Multiply(A=StaticSwitchParameter(True=@n35, False=@n26) {parameterName=VerticalLineSwitch}, B=@n32.A))
 ]
}
