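# Material function for the Fearless Flight Hero backpack that outputs an animated hex-pattern Colour and an Alpha taken from the hex texture.
# A baked UV texture, jittered by cloud noise, is offset by a time input with flip-controlled direction, forming sharp moving lines scaled by LinesGain.
# A second scroll at hexTimeGain samples a random-colour hex pulse texture as a faint additive layer.
# A named reroute quantises Time to an FPS parameter for stepped animation.
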MaterialFunction /BRCosmetics/Accessories/FORT_Backpacks/Backpack_M_MED_FearlessFlightHero/Materials/MF_Hex
graph {nodes [
  Constant2Vector() {g=1, r=1}
  n1 = TextureCoordinate()
  n2 = TextureSample!(
   UVs=FunctionInput(Preview=@n1) {bUsePreviewValueAsDefault=true, inputName=UV, inputType=FunctionInput_Vector2}
   Tex
   Bias
   Apply View MipBias
  ) {
   mipValueMode=TMVM_MipBias
   texture=/BRCosmetics/Accessories/FORT_Backpacks/Backpack_M_MED_FearlessFlightHero/Textures/04_Portal_Large_Hex_WholeSqr
  }
  n3 = LinearInterpolate(A=Constant3Vector() {constant=<rgba 1 0.05 0 0>}, B=@n2.RGB, Alpha=ScalarParameter() {defaultValue=0.25, parameterName=Constrast})
  n4 = Desaturation(None=@n2.RGB, Fraction=Constant() {r=1})
  n5 = Constant() {r=2.71}
  n6 = TextureSampleParameter2D!(
   UVs=FunctionInput(Preview=@n1) {bUsePreviewValueAsDefault=true, inputName=UV_Static, inputType=FunctionInput_Vector2}
   Level
   Apply View MipBias
  ) {
   mipValueMode=TMVM_MipLevel
   parameterName=Param
   samplerType=SAMPLERTYPE_LinearColor
   texture=/BRCosmetics/Accessories/FORT_Backpacks/Backpack_M_MED_FearlessFlightHero/Textures/M_bakeHexUV_Baked
  }
  n7 = ComponentMask(None=@n6.RGB) {r=true}
  n8 = ComponentMask(None=@n6.RGB) {g=true}
  n9 = FunctionInput() {bUsePreviewValueAsDefault=true, inputName=time_gain, inputType=FunctionInput_Vector2, previewValue=<0.25 0 0 1>}
  n10 = Multiply(A=NamedRerouteUsage(), B=@n9) {constB=0.1}
  n11 = FunctionInput() {bUsePreviewValueAsDefault=true, inputName=flip, inputType=FunctionInput_Scalar, previewValue=<1 0 0 1>}
  n12 = Multiply(
   A=FunctionInput(Preview=MaterialFunctionCall(Y=@n10) {materialFunction=&MakeFloat2}.Result) {bUsePreviewValueAsDefault=true, inputName=time, inputType=FunctionInput_Vector2}
   B=@n11
  )
  n13 = Saturate(None=Sign(None=ComponentMask(None=@n12) {g=true}))
  n14 = LinearInterpolate(A=@n8, B=OneMinus(None=@n8), Alpha=@n13)
  n15 = ComponentMask(None=@n6.RGB) {g=true, r=true}
  n16 = Constant2Vector() {g=1, r=2}
  n17 = Subtract(A=TextureSample!(UVs=Multiply(A=@n15, B=@n16), Tex, Apply View MipBias) {texture=/EpicBaseTextures/Noises/T_Clouds_Noise3_D}.R) {constB=0.5}
  n18 = AppendVector(A=@n7, B=Add(A=@n14, B=Multiply(A=@n17) {constB=0.1}))
  n19 = Abs(None=Subtract(A=@n18, B=@n12))
  n20 = Add(A=@n18, B=@n12)
  n21 = Frac(None=Multiply(A=LinearInterpolate(A=@n19, B=@n20, Alpha=@n13)) {constB=2.71})
  n22 = OneMinus(None=Saturate(None=ComponentMask(None=@n21) {g=true}))
  n23 = Saturate(None=Multiply(
    A=MaterialFunctionCall(In=@n4, Contrast=@n5) {materialFunction=&CheapContrast}.Result
    B=Power(Base=@n22) {constExponent=12}
   ))
  n24 = Constant() {r=5}
  MaterialExpressionRemap_1 = MaterialXRemap!(
   None=@n23
   InputLow
   InputHigh
   TargetLow
   TargetHigh=FunctionInput(Preview=@n24) {bUsePreviewValueAsDefault=true, inputName=LinesGain, inputType=FunctionInput_Scalar}
  ) {targetHighDefault=5, targetLowDefault=0.75}
  n25 = NamedRerouteUsage()
  n26 = Saturate(None=ComponentMask(None=Frac(None=@n25)) {g=true})
  n27 = Multiply(A=AppendVector(A=NamedRerouteUsage(), B=OneMinus(None=@n26)), B=Constant2Vector() {g=8, r=2})
  n28 = Add(
   A=Multiply(A=@n3, B=@MaterialExpressionRemap_1)
   B=Multiply(A=TextureSample(UVs=@n27) {samplerType=SAMPLERTYPE_Masks, texture=/Game/Creative/Sets/PropSets/Primitives/Customization/Textures/Hex/T_CM_HexPulse_RandomColor}.RGB) {constB=0.015}
  )
  FunctionOutput(None=Max(A=@n28) {constB=0}) {outputName=Colour}
  FunctionOutput(None=Saturate(None=Multiply(A=@n2.A) {constB=1.25})) {outputName=Alpha, sortPriority=1}
  LinearInterpolate(A=@n3, B=Multiply(A=Constant3Vector() {constant=<rgba 1 0.6324 0.3906 0>}, B=@n24), Alpha=@n23)
  n29 = AppendVector(A=@n7, B=@n14)
  n30 = FunctionInput() {bUsePreviewValueAsDefault=true, inputName=hexTimeGain, inputType=FunctionInput_Scalar, previewValue=<0.5 0 0 1>}
  n31 = Multiply(A=@n12, B=@n30) {constB=0.5}
  n32 = Subtract(A=@n29, B=@n31)
  NamedRerouteDeclaration(None=LinearInterpolate(A=Abs(None=@n32), B=Add(A=@n29, B=@n31), Alpha=@n13)) {name=normal UVs, nodeColor=<rgba 0 1 0.4609 1>}
  NamedRerouteDeclaration(None=@n7) {name=outeru, nodeColor=<rgba 1 0.4922 0 1>}
  n33 = ScalarParameter() {defaultValue=12, parameterName=FPS}
  n34 = Multiply(A=Time(), B=@n33) {constB=12}
  NamedRerouteDeclaration(None=Divide(A=Floor(None=@n34), B=@n33) {constB=12}) {name=fps, nodeColor=<rgba 1 0 0.4688 1>}
  Saturate(None=@n28)
  Time()
 ]}
