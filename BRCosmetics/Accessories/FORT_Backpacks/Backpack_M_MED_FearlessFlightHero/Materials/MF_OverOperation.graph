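# Material function in the Fearless Flight Hero backpack folder that composites color layer A over color layer B with the alpha "over" operation.
# It weights ColA by AlphaA and ColB by AlphaB times the inverse of AlphaA, then sums them.
# The combined coverage becomes the Alpha output and divides that sum to give an un-premultiplied Color, with no guard against zero alpha.
# Inputs default to black colors at full alpha.
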MaterialFunction /BRCosmetics/Accessories/FORT_Backpacks/Backpack_M_MED_FearlessFlightHero/Materials/MF_OverOperation
graph {nodes [
  n1 = FunctionInput(Preview=Constant() {r=1}) {bUsePreviewValueAsDefault=true, inputName=AlphaA, inputType=FunctionInput_Scalar, sortPriority=1}
  n2 = Multiply(A=FunctionInput(Preview=Constant3Vector()) {bUsePreviewValueAsDefault=true, inputName=ColA}, B=@n1)
  n3 = FunctionInput(Preview=Constant() {r=1}) {bUsePreviewValueAsDefault=true, inputName=AlphaB, inputType=FunctionInput_Scalar, sortPriority=3}
  n4 = Constant3Vector()
  n5 = OneMinus(None=@n1)
  n6 = Multiply(A=Multiply(A=FunctionInput(Preview=@n4) {bUsePreviewValueAsDefault=true, inputName=ColB, sortPriority=2}, B=@n3), B=@n5)
  n7 = Add(A=@n1, B=Multiply(A=@n3, B=@n5))
  FunctionOutput(None=Divide(A=Add(A=@n2, B=@n6), B=@n7)) {outputName=Color}
  FunctionOutput(None=@n7) {outputName=Alpha, sortPriority=1}
 ]}
