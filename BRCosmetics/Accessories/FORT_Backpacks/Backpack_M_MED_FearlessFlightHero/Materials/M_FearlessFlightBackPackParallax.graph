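# Two-sided material for the Fearless Flight Hero backpack that renders an emissive hexagonal portal with fake depth.
# Five MF_Hex layers are parallax-offset by tangent-space camera vector, rotated by time or optional custom primitive data, and composited.
# Behind them, two ray-traced spheres map a magenta portal texture as the background.
# The composite, compensated for eye adaptation, drives emissive on a black base through the shared character opacity function.
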Material /BRCosmetics/Accessories/FORT_Backpacks/Backpack_M_MED_FearlessFlightHero/Materials/M_FearlessFlightBackPackParallax
graph {
 properties {
  bUseMaterialAttributes=true
  bUsedWithMeshDeformer=true
  bUsedWithSkeletalMesh=true
  pixelDepthOffsetMode=PDOM_Legacy
  twoSided=true
  usedShadingModels=Default Lit
 }
 nodes [
  n1 = NamedRerouteUsage()
  n2 = ScalarParameter() {parameterName=Depth}
  n3 = NamedRerouteUsage()
  Add(A=@n1, B=Multiply(A=NamedRerouteUsage(), B=Divide(A=@n2, B=@n3)))
  n4 = ScalarParameter() {defaultValue=0.2537, parameterName=scale5}
  n5 = ScalarParameter() {defaultValue=1.7798, parameterName=Depth5}
  n6 = NamedRerouteUsage()
  n7 = Add(
   A=MaterialFunctionCall(UVs=NamedRerouteUsage(), Texture Scale=@n4) {materialFunction=&ScaleUVsByCenter}.UVs
   B=Multiply(A=NamedRerouteUsage(), B=Divide(A=@n5, B=@n6))
  )
  n8 = NamedRerouteUsage()
  n9 = StaticSwitchParameter(
   True=NamedRerouteUsage()
   False=Multiply(A=ScalarParameter() {defaultValue=-3.2652, parameterName=Rotation5}, B=Multiply(A=@n8) {constB=-0.024})
  ) {parameterName=UseCustomPrimAnim}
  n10 = NamedRerouteUsage()
  n11 = Saturate(None=MaterialFunctionCall(UVs=@n7, "Rotation Angle (0-1)"=Add(A=@n9, B=@n10) {constB=0.032}) {materialFunction=&CustomRotator}.Rotated Values)
  n12 = TextureSample(UVs=@n11) {
   samplerType=SAMPLERTYPE_LinearColor
   texture=/BRCosmetics/Accessories/FORT_Backpacks/Backpack_M_MED_FearlessFlightHero/Textures/M_bakeHexUV_Baked
  }
  n13 = NamedRerouteUsage()
  n14 = Add(A=ComponentMask(None=@n12.RGB) {g=true, r=true}, B=@n13)
  n15 = TextureSample!(
   UVs=Multiply(A=Frac(None=@n14), B=NamedRerouteUsage())
   Tex
   "DDX(UVs)"=NamedRerouteUsage()
   "DDY(UVs)"=NamedRerouteUsage()
   Apply View MipBias
  ) {
   mipValueMode=TMVM_Derivative
   samplerType=SAMPLERTYPE_Masks
   texture=/Game/Creative/Sets/PropSets/Primitives/Customization/Textures/Hex/T_CM_HexPulse_RandomColor
  }
  Add(
   A=TextureSample!(UVs=@n11, Tex, Apply View MipBias) {texture=/BRCosmetics/Accessories/FORT_Backpacks/Backpack_M_MED_FearlessFlightHero/Textures/04_Portal_Large_Hex_WholeSqr}.RGB
   B=Multiply(A=@n15.R, B=ScalarParameter() {defaultValue=0.25, parameterName=RingMovementGain})
  )
  ComponentMask(None=MaterialFunctionCall() {materialFunction=&BoundingBoxBased_0-1_UVW}.RGB) {b=true, r=true}
  Constant2Vector() {g=1, r=3}
  Constant()
  Constant() {r=-1}
  Constant() {r=12}
  Constant() {r=12}
  Constant() {r=12}
  Constant() {r=12}
  Constant() {r=12}
  n16 = NamedRerouteUsage()
  n17 = MaterialFunctionCall!(
   time=NamedRerouteUsage()
   UV=Saturate(None=MaterialFunctionCall(UVs=@n1, "Rotation Angle (0-1)"=@n16) {materialFunction=&CustomRotator}.Rotated Values)
   LinesGain
   time_gain
   UV_Static
   hexTimeGain
   flip
  ) {materialFunction=&/MF_Hex}
  n18 = Constant() {r=1}
  n19 = ScalarParameter() {defaultValue=0.1667, parameterName=Depth2}
  n20 = NamedRerouteUsage()
  n21 = Add(
   A=MaterialFunctionCall(UVs=NamedRerouteUsage(), Texture Scale=Subtract(A=@n18, B=@n19)) {materialFunction=&ScaleUVsByCenter}.UVs
   B=Multiply(A=NamedRerouteUsage(), B=Divide(A=@n19, B=@n20))
  )
  n22 = ScalarParameter() {defaultValue=1.1818, parameterName=Rotation2}
  n23 = Multiply(A=NamedRerouteUsage()) {constB=-0.064}
  n24 = Add(A=StaticSwitchParameter(True=NamedRerouteUsage(), False=Multiply(A=@n22, B=@n23)) {parameterName=UseCustomPrimAnim}, B=NamedRerouteUsage()) {constB=0.032}
  n25 = MaterialFunctionCall(
   time=NamedRerouteUsage()
   UV=Saturate(None=MaterialFunctionCall(UVs=@n21, "Rotation Angle (0-1)"=@n24) {materialFunction=&CustomRotator}.Rotated Values)
   UV_Static=@n21
   flip=Multiply(A=NamedRerouteUsage())
  ) {materialFunction=&MF_Hex}
  n26 = MaterialFunctionCall!(ColA=Multiply(A=@n17.Colour, B=Constant() {r=4}), AlphaA=@n17.Alpha, ColB=Multiply(A=@n25.Colour, B=Constant() {r=3.2}), AlphaB=@n25.Alpha) {materialFunction=&/MF_OverOperation}
  n27 = Constant() {r=1}
  n28 = ScalarParameter() {defaultValue=0.3813, parameterName=Depth3}
  n29 = NamedRerouteUsage()
  n30 = Add(
   A=MaterialFunctionCall(UVs=NamedRerouteUsage(), Texture Scale=Subtract(A=@n27, B=@n28)) {materialFunction=&ScaleUVsByCenter}.UVs
   B=Multiply(A=NamedRerouteUsage(), B=Divide(A=@n28, B=@n29))
  )
  n31 = ScalarParameter() {defaultValue=1.6583, parameterName=Rotation3}
  n32 = Multiply(A=NamedRerouteUsage()) {constB=0.036}
  n33 = Add(A=StaticSwitchParameter(True=NamedRerouteUsage(), False=Multiply(A=@n31, B=@n32)) {parameterName=UseCustomPrimAnim}, B=NamedRerouteUsage()) {constB=0.032}
  n34 = MaterialFunctionCall(
   time=NamedRerouteUsage()
   UV=Saturate(None=MaterialFunctionCall(UVs=@n30, "Rotation Angle (0-1)"=@n33) {materialFunction=&CustomRotator}.Rotated Values)
   UV_Static=@n30
   flip=Multiply(A=NamedRerouteUsage()) {constB=-1}
  ) {materialFunction=&MF_Hex}
  n35 = MaterialFunctionCall(ColA=@n26.Color, AlphaA=@n26.Color, ColB=Multiply(A=@n34.Colour, B=Constant() {r=2.4}), AlphaB=@n34.Alpha) {materialFunction=&MF_OverOperation}
  n36 = ScalarParameter() {defaultValue=0.384, parameterName=scale4}
  n37 = ScalarParameter() {defaultValue=0.836, parameterName=Depth4}
  n38 = NamedRerouteUsage()
  n39 = Add(
   A=MaterialFunctionCall(UVs=NamedRerouteUsage(), Texture Scale=@n36) {materialFunction=&ScaleUVsByCenter}.UVs
   B=Multiply(A=NamedRerouteUsage(), B=Divide(A=@n37, B=@n38))
  )
  n40 = ScalarParameter() {defaultValue=2.9432, parameterName=Rotation4}
  n41 = Multiply(A=NamedRerouteUsage()) {constB=-0.024}
  n42 = Add(A=StaticSwitchParameter(True=NamedRerouteUsage(), False=Multiply(A=@n40, B=@n41)) {parameterName=UseCustomPrimAnim}, B=NamedRerouteUsage()) {constB=0.032}
  n43 = MaterialFunctionCall(
   time=NamedRerouteUsage()
   UV=Saturate(None=MaterialFunctionCall(UVs=@n39, "Rotation Angle (0-1)"=@n42) {materialFunction=&CustomRotator}.Rotated Values)
   flip=Multiply(A=NamedRerouteUsage()) {constB=-1}
  ) {materialFunction=&MF_Hex}
  n44 = MaterialFunctionCall(ColA=@n35.Color, AlphaA=@n35.Color, ColB=Multiply(A=@n43.Colour, B=Constant() {r=1.8}), AlphaB=@n43.Alpha) {materialFunction=&MF_OverOperation}
  n45 = MaterialFunctionCall(time=@n13, UV=@n11, UV_Static=@n7) {materialFunction=&MF_Hex}
  n46 = MaterialFunctionCall(ColA=@n44.Color, AlphaA=@n44.Color, ColB=Multiply(A=@n45.Colour, B=Constant() {r=1}), AlphaB=@n45.Alpha) {materialFunction=&MF_OverOperation}
  Desaturation(
   None=MaterialFunctionCall!(In=@n46.Color, Contrast=ScalarParameter() {parameterName=Constrast}) {materialFunction=/Engine/Functions/Engine_MaterialFunctions01/ImageAdjustment/CheapContrast_RGB}.Result
   Fraction=Constant() {r=-0.4}
  )
  EyeAdaptation()
  n47 = Transform(None=CameraVectorWS()) {transformSourceType=TRANSFORMSOURCE_World, transformType=TRANSFORM_Tangent}
  MaterialFunctionCall!(Normal=Constant3Vector() {constant=<rgba 0 0 1 0>}, CameraVector=@n47) {materialFunction=/Engine/Functions/Engine_MaterialFunctions02/Utility/CustomReflectionVector}
  MaterialFunctionCall(Rotation Center=Constant(), "Rotation Angle (0-1)"=Multiply(A=Time()) {constB=0.1}) {materialFunction=&CustomRotator}
  n48 = Constant()
  n49 = Length(None=PreSkinnedLocalBounds().Extents) {is=class MaterialExpressionLength}
  n50 = Multiply(A=Transform(None=CameraVectorWS()) {transformSourceType=TRANSFORMSOURCE_World, transformType=TRANSFORM_Tangent}) {constB=-1}
  n51 = Transform(None=VertexInterpolator(VS=PreSkinnedPosition()).PS) {transformSourceType=TRANSFORMSOURCE_Local, transformType=TRANSFORM_Tangent}
  n52 = MaterialFunctionCall!(Sphere Position, Sphere Radius=Multiply(A=@n49) {constB=0.8}, RayDirection=@n50, World position=@n51) {materialFunction=/Engine/Functions/Engine_MaterialFunctions02/RayTracedSphere}
  n53 = Constant3Vector() {constant=<rgba 0 0 -1 0>}
  n54 = Constant3Vector() {constant=<rgba 0 1 0 0>}
  n55 = Add(A=MaterialFunctionCall!(Vector=MaterialFunctionCall!(VectorToRotate=@n52.Intersection Surface Normal, LookAtVector=@n53, RestingVector=@n54) {materialFunction=/Engine/Functions/Engine_MaterialFunctions01/Vectors/RotateVector}.Result) {materialFunction=/Engine/Functions/Engine_MaterialFunctions01/Coordinates/LongLatToUV}.Result) {constB=0.25}
  n56 = MaterialFunctionCall(Sphere Radius=Multiply(A=@n49) {constB=0.5}, RayDirection=@n50, World position=@n51) {materialFunction=&RayTracedSphere}
  n57 = Add(A=MaterialFunctionCall(Vector=MaterialFunctionCall(VectorToRotate=@n56.Intersection Surface Normal, LookAtVector=@n53, RestingVector=@n54) {materialFunction=&RotateVector}.Result) {materialFunction=&LongLatToUV}.Result) {constB=0.5}
  n58 = Multiply(A=LinearInterpolate(
    A=TextureSample(UVs=@n55) {texture=/BRCosmetics/Accessories/FORT_Backpacks/Backpack_M_MED_FearlessFlightHero/Textures/10_Portal_Background_Tile_Magenta}.RGB
    B=TextureSample(UVs=@n57) {texture=&10_Portal_Background_Tile_Magenta}.RGB
   )) {constB=2}
  n59 = Constant() {r=1}
  n60 = MaterialFunctionCall(ColA=Max(A=@n46.Color) {constB=0}, AlphaA=@n46.Alpha, ColB=@n58, AlphaB=@n59) {materialFunction=&MF_OverOperation}
  n61 = EyeAdaptationInverse()
  MaterialFunctionCall!(Material=MakeMaterialAttributes!(
    BaseColor=@n48
    Metallic=@n48
    Specular=@n48
    Roughness=Constant() {r=1}
    Anisotropy
    EmissiveColor=Multiply(A=@n60.Color, B=@n61.EyeAdaptationInverse)
    Opacity
    OpacityMask
    Normal
    Tangent
    WorldPositionOffset
    SubsurfaceColor
    ClearCoat
    ClearCoatRoughness
    AmbientOcclusion
    Refraction
    CustomizedUVs_0
    CustomizedUVs_1
    CustomizedUVs_2
    CustomizedUVs_3
    CustomizedUVs_4
    CustomizedUVs_5
    CustomizedUVs_6
    CustomizedUVs_7
    PixelDepthOffset
    ShadingModel
    Displacement
   )) {materialFunction=/Game/Packages/Fortress_SharedMaterials/Base_Material_Functions/BaseCharacterShaderFunctions/BaseFN_CharacterOpacity}
  MaterialFunctionCall(ColA=@n46.Color, AlphaA=@n46.Color, ColB=@n58, AlphaB=@n59) {materialFunction=&MF_OverOperation}
  n62 = Multiply(A=@n60.Color) {constB=0.5}
  MaterialFunctionCall(Color_Morning=@n60.Color, Color_Day=@n60.Color, Color_Evening=@n62, Color_Night=@n62) {materialFunction=&Color_Time_of_Day}
  n63 = TextureCoordinate()
  n64 = MaterialFunctionCall(UVs=@n63, "Rotation Angle (0-1)"=Constant() {r=0.082}) {materialFunction=&CustomRotator}
  MaterialExpressionSubtract_4 = Subtract(A=Multiply(A=@n64.Rotated Values) {constB=2}, B=Constant2Vector() {g=-0.5, r=0.14})
  n65 = Custom!(position=@MaterialExpressionSubtract_4) {
   code="\r\n    position /= float2(sqrt(3.0), 1.5);\r\n   position.y -= 0.5;\r\n    position.x -= frac(floor(position.y) * 0.5);\r\n    position = abs(frac(position) - 0.5);\r\n    return abs(1.0 - max(position.x + position.y * 1.5, position.x * 2.0)) * sqrt(3.0) * 0.5;\r\n"
   outputType=CMOT_Float2
   inputs [{
     inputName=position
     input {expression=&self:MaterialExpressionSubtract_4}
    }]
  }
  n66 = Multiply(A=ComponentMask(None=@n65) {r=true}) {constB=3}
  MaterialExpressionRemap_0 = MaterialXRemap!(None=@n65, InputLow, InputHigh, TargetLow, TargetHigh) {targetHighDefault=0.2, targetLowDefault=3}
  n67 = MaterialFunctionCall(Float2=Multiply(A=Subtract(A=@n64.Rotated Values) {constB=0.5}, B=@MaterialExpressionRemap_0)) {materialFunction=&BreakOutFloat2Components}
  n68 = Arctangent2!(Y=@n67.R, X=@n67.R)
  n69 = MaterialFunctionCall(Multiplier=Constant() {r=2}) {materialFunction=&Pi}
  n70 = Time()
  n71 = Frac(None=Add(A=Divide(A=@n68, B=@n69.Result), B=Multiply(A=@n70) {constB=0.1}))
  n72 = Multiply(A=@n71) {constB=2}
  n73 = OneMinus(None=@n71)
  Multiply(
   A=TextureSample(UVs=AppendVector(A=@n66, B=@n72)) {texture=/Game/Athena/Prototype/Blueprints/Sneezy/Textures/T_Sneezy_MagneticWaves}.RGB
   B=Saturate(None=Power(Base=@n73) {constExponent=6})
  )
  Multiply(A=@n63)
  NamedRerouteDeclaration(None=TextureCoordinate()) {name=UV, nodeColor=<rgba 1 0.3281 0 1>}
  NamedRerouteDeclaration(None=ComponentMask(None=@n47) {g=true, r=true}) {name=Offset2D, nodeColor=<rgba 0 1 0.5547 1>}
  n74 = NamedRerouteUsage()
  n75 = ScalarParameter() {defaultValue=0.2, parameterName=RingMotionSpeed}
  NamedRerouteDeclaration(None=MaterialFunctionCall(Y=Multiply(A=@n74, B=@n75) {constB=0.1}) {materialFunction=&MakeFloat2}.Result) {name=timering, nodeColor=<rgba 0 0.8359 1 1>}
  NamedRerouteDeclaration(None=Constant2Vector() {g=4, r=12}) {name=hex UV tile, nodeColor=<rgba 1 0 0.3984 1>}
  NamedRerouteDeclaration(None=Sign(None=@n23)) {name=sign1, nodeColor=<rgba 1 0.75 0 1>}
  NamedRerouteDeclaration(None=Sign(None=@n32)) {name=sign2, nodeColor=<rgba 1 0.75 0 1>}
  n76 = ScalarParameter() {defaultValue=12, parameterName=FPS}
  n77 = Multiply(A=Time(), B=@n76) {constB=12}
  NamedRerouteDeclaration(None=Divide(A=Floor(None=@n77), B=@n76) {constB=12}) {name=fps, nodeColor=<rgba 1 0 0.4687 1>}
  n78 = TwoSidedSign()
  NamedRerouteDeclaration(None=Multiply(A=ComponentMask(None=@n47) {b=true}, B=Multiply(A=@n78) {constB=-1})) {name=OffsetZ, nodeColor=<rgba 0.7812 1 0 1>}
  n79 = VectorParameter() {bUseCustomPrimitiveData=true, parameterName=Hex Rotations Custom Prim}
  NamedRerouteDeclaration(None=@n79.R) {name=Hex Anim 1, nodeColor=<rgba 0 1 0.9531 1>}
  NamedRerouteDeclaration(None=@n79.G) {name=Hex Anim 2, nodeColor=<rgba 0 1 0.9531 1>}
  NamedRerouteDeclaration(None=@n79.B) {name=Hex Anim 3, nodeColor=<rgba 0 1 0.9531 1>}
  NamedRerouteDeclaration(None=@n79.A) {name=Hex Anim 4, nodeColor=<rgba 0 1 0.9531 1>}
  NamedRerouteDeclaration(None=ScalarParameter() {defaultValue=0.034, parameterName=Hex Offset}) {name=hex offset, nodeColor=<rgba 0.6641 0 1 1>}
  NamedRerouteDeclaration(None=DDX(Value=@n1)) {name=DDX, nodeColor=<rgba 0 1 0.8828 1>}
  NamedRerouteDeclaration(None=DDY(Value=@n1)) {name=DDY, nodeColor=<rgba 0.8984 1 0 1>}
  OneMinus(None=@n65)
  OneMinus()
  Subtract(A=@n36, B=@n37)
  Subtract(A=@n4, B=@n5)
  TextureSample() {texture=&T_Sneezy_MagneticWaves}
  Time()
  Time()
  Time()
  Time()
 ]
}
